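; NPC conversation task node that sells items from a loot tier group for a currency cost.
; IsRequirementSatisfied checks the player's balance after the parent check, keeping an unaffordable offer visible but failed.
; GatherStaticExtraData publishes the first rolled item, cost, remaining count and stack size as name/value pairs.
; ExecuteTaskNode rolls the loot again independently, grants one of each dropped item, then removes the currency only if items were granted.
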(fn AddNodePairToExtraData
  (params (Name :type String) (Value :type String))
  (CallFunction|AddSingleEntrytoArray (Utilities|Struct|MakeConversationNodeParameterPair Name Value))
  (return))

(fn AddSingleEntryToArray
  (params (entry :type Conversation%20Node%20Parameter%20Pair%20Structure%20%28by%20ref%29))
  (bind _extradata (Variables|Default|GetextraData))
  (if (Utilities|Array|ContainsItem _extradata entry)
    (return _extradata)
    (else
      (Utilities|Array|Add _extradata entry)
      (return _extradata))))

(fn AddtemFromLootTierGroup
  (params (playerController :type Fort%20Player%20Controller%20Object%20Reference))
  (bind (_outloottodrop _returnvalue) (Loot|PickLootDrops (Variables|ItemData|GetLoot%20Tier%20Group%20Name) (Loot|GetLootLevel) -1 "(GameplayTags=)"))
  (if _returnvalue
    (for _array_element _outloottodrop
      (bind _asfort_item_definition (Utilities|Casting|CastToFortItemDefinition (Class|FortItemFunctionLibrary|BreakFortItemEntry _array_element)))
      (InventoryManagement|AddItem playerController _asfort_item_definition 1)
      (LocalVariables|SetbValidItem true))
    (if (LocalVariables|GetbValidItem)
      (return true)
      (else
        _))
    (else
      _))
  (return false))

(fn CheckPlayerBalance
  (params (TargetPlayerController :type Fort%20Player%20Controller%20Object%20Reference) (currencyItemDefinition :type Fort%20World%20Item%20Definition%20Object%20Reference))
  (bind _returnvalue (InventoryManagement|GetItemCountforItemDefinition TargetPlayerController currencyItemDefinition))
  (return (>= _returnvalue (ScalableFloat|GetValueasInteger (Variables|Pricing|GetserviceCost)))))

(fn ExecuteTaskNode
  (params (Context :type Conversation%20Context%20Structure%20%28by%20ref%29))
  (bind _returnvalue (|CallParentFunction Context))
  (bind (_payload _success) (Default|GetPlayerController Context))
  (if _success
    (bind _blootdropsvalid (Default|AddtemfromLootTierGroup _payload))
    (if _blootdropsvalid
      (Default|RemoveResourcefromPlayer _payload)
      (return _returnvalue)
      (else
        _))
    (else
      _))
  (return _returnvalue))

(fn GatherStaticExtraData
  (params (Context :type Conversation%20Context%20Structure%20%28by%20ref%29))
  (bind _extradata (Variables|Default|GetextraData))
  (bind _inoutextradata (|CallParentFunction Context))
  (Utilities|Array|Clear (Variables|Default|GetextraData))
  (bind (_blootdropsvalid _fortitemdefinition) (CallFunction|GetSaleItemData))
  (if _blootdropsvalid
    (CallFunction|AddNodePairtoExtraData "item" (Utilities|GetObjectPathString _fortitemdefinition))
    (else
      _))
  (CallFunction|GetPricingData)
  (CallFunction|AddNodePairtoExtraData "remaining" "1")
  (CallFunction|AddNodePairtoExtraData "stacksize" "1")
  (Utilities|Array|AppendArray _extradata _inoutextradata)
  (return _extradata))

(fn GetPlayerController
  (params (Context :type Conversation%20Context%20Structure%20%28by%20ref%29))
  (bind _returnvalue (Class|FortConversationContextLibrary|GetParticipantActorPlayer Context))
  (Utilities|IsValid _returnvalue
    (:Is%20Valid
      (bind _asfort_player_controller (Utilities|Casting|CastToFortPlayerController _returnvalue)
        (:then
          (return _asfort_player_controller true))
        (:CastFailed
          _)))
    (:Is%20Not%20Valid
      _))
  (return 0 false))

(fn GetPricingData
  (CallFunction|AddNodePairtoExtraData "cost" (Utilities|String|ToString(Integer) (ScalableFloat|GetValueasInteger (Variables|Pricing|GetserviceCost)))))

(fn GetSaleItemData
  (bind (_outloottodrop _returnvalue) (Loot|PickLootDrops (Variables|ItemData|GetLoot%20Tier%20Group%20Name) (Loot|GetLootLevel) -1 "(GameplayTags=)"))
  (if _returnvalue
    (for _array_element _outloottodrop
      (bind _asfort_item_definition (Utilities|Casting|CastToFortItemDefinition (Class|FortItemFunctionLibrary|BreakFortItemEntry _array_element))
        (:then
          (Utilities|IsValid _asfort_item_definition
            (:Is%20Valid
              (return true _asfort_item_definition _array_element))
            (:Is%20Not%20Valid
              _)))
        (:CastFailed
          _))
      (return false 0 0))
    (else
      _)))

(fn IsRequirementSatisfied
  (params (Context :type Conversation%20Context%20Structure%20%28by%20ref%29))
  (bind _returnvalue (|CallParentFunction Context))
  (if (Utilities|Enum|Equal(Enum) _returnvalue "Passed")
    (CallFunction|AddNodePairtoExtraData "locked" "0")
    (bind (_payload _success) (Default|GetPlayerController Context))
    (if _success
      (bind _bhascurrency (CallFunction|CheckPlayerBalance _payload (Variables|Pricing|GetcurrencyItemDefinition)))
      (if _bhascurrency
        (return "Passed")
        (else
          (return "FailedButVisible")))
      (else
        _))
    (elif (Variables|Requirements|GetbHideWhenFailed)
      _
      (else
        (return _returnvalue))))
  (return "FailedAndHidden"))

(fn RemoveResourceFromPlayer
  (params (PlayerController :type Fort%20Player%20Controller%20Object%20Reference))
  (Inventory|RemoveItemFromPlayer :PlayerController PlayerController :ItemDefinition (Variables|Pricing|GetcurrencyItemDefinition) :AmountToRemove (ScalableFloat|GetValueasInteger (Variables|Pricing|GetserviceCost)) :bForceRemoval false :bSkipVerbRemoval false)
  (return))
